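# HUD widget for one of six rocks in the Frog Party LTM status display, showing an undiscovered diamond or the gem.
# PreConstruct uses StoneNumber to choose the gem material and diamond fill color; Construct and EventHide show the undiscovered state.
# Intro reveals the gem with a scale and whiteout/shine animation.
# Outro layers a shared scale-and-fade animation with a StoneNumber-selected translation, sending each gem in its own direction.
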WidgetBlueprint /Figment_Martin/Athena/HUD/LTM/FrogParty/HUD_FrogPartyStatus_Rock
animations [
 Arrive {
  binding_count=1
  length_seconds=1.2542
  object=&self:Arrive
  sequence {bindings [Image_Gem {
     id=#1
     possessed_class=Image
     tracks [
      {
       type=MovieScene2DTransformTrack
       sections [{
         type=MovieScene2DTransformSection
         channels [
          Translation.X {type=MovieSceneScriptingFloatChannel}
          Translation.Y {type=MovieSceneScriptingFloatChannel}
          Angle {type=MovieSceneScriptingFloatChannel}
          Scale.X {type=MovieSceneScriptingFloatChannel, keys [[0,2],[4,0.8251],[5,1]]}
          Scale.Y {type=MovieSceneScriptingFloatChannel, keys [[0,2],[4,0.8251],[5,1]]}
          Shear.X {type=MovieSceneScriptingFloatChannel}
          Shear.Y {type=MovieSceneScriptingFloatChannel}
         ]
        }]
      }
      {
       type=MovieSceneWidgetMaterialTrack
       sections [{
         type=MovieSceneParameterSection
         channels [
          Whiteout {type=MovieSceneScriptingFloatChannel, keys [[0,1],[4,0]]}
          ShineOffset {type=MovieSceneScriptingFloatChannel, keys [[5,-0.54],[25.0837,0]]}
         ]
        }]
      }
     ]
    }]}
 }
 Collected1 {
  binding_count=1
  length_seconds=1.1226
  object=&self:Collected1
  sequence {bindings [Image_Gem {
     id=#2
     possessed_class=Image
     tracks [{
       type=MovieScene2DTransformTrack
       sections [{
         type=MovieScene2DTransformSection
         end_frame=22
         start_frame=0
         channels [
          Translation.X {type=MovieSceneScriptingFloatChannel, keys [[0,0],[22.451,-100]]}
          Translation.Y {type=MovieSceneScriptingFloatChannel, keys [[0,0],[22.451,60]]}
          Angle {type=MovieSceneScriptingFloatChannel}
          Scale.X {type=MovieSceneScriptingFloatChannel}
          Scale.Y {type=MovieSceneScriptingFloatChannel}
          Shear.X {type=MovieSceneScriptingFloatChannel}
          Shear.Y {type=MovieSceneScriptingFloatChannel}
         ]
        }]
      }]
    }]}
 }
 CollectedBase {
  binding_count=1
  length_seconds=1.1226
  object=&self:CollectedBase
  sequence {bindings [Image_Gem {
     id=#2
     possessed_class=Image
     tracks [
      {
       type=MovieScene2DTransformTrack
       sections [{
         type=MovieScene2DTransformSection
         end_frame=22
         start_frame=0
         channels [
          Translation.X {type=MovieSceneScriptingFloatChannel}
          Translation.Y {type=MovieSceneScriptingFloatChannel}
          Angle {type=MovieSceneScriptingFloatChannel}
          Scale.X {type=MovieSceneScriptingFloatChannel, keys [[0,1],[6.5423,2],[22.451,1]]}
          Scale.Y {type=MovieSceneScriptingFloatChannel, keys [[0,1],[6.5423,2],[22.451,1]]}
          Shear.X {type=MovieSceneScriptingFloatChannel}
          Shear.Y {type=MovieSceneScriptingFloatChannel}
         ]
        }]
      }
      {
       type=MovieSceneFloatTrack
       display_name=Render Opacity
       sections [{
         type=MovieSceneFloatSection
         end_frame=22
         start_frame=0
         channels [None {type=MovieSceneScriptingFloatChannel, keys [[0,1],[12.8913,1],[22.451,0]]}]
        }]
      }
     ]
    }]}
 }
 Collected2 {
  binding_count=1
  length_seconds=1.1226
  object=&self:Collected2
  sequence {bindings [Image_Gem {
     id=#2
     possessed_class=Image
     tracks [{
       type=MovieScene2DTransformTrack
       sections [{
         type=MovieScene2DTransformSection
         end_frame=22
         start_frame=0
         channels [
          Translation.X {type=MovieSceneScriptingFloatChannel, keys [[0,0],[22.451,100]]}
          Translation.Y {type=MovieSceneScriptingFloatChannel, keys [[0,0],[22.451,-60]]}
          Angle {type=MovieSceneScriptingFloatChannel}
          Scale.X {type=MovieSceneScriptingFloatChannel}
          Scale.Y {type=MovieSceneScriptingFloatChannel}
          Shear.X {type=MovieSceneScriptingFloatChannel}
          Shear.Y {type=MovieSceneScriptingFloatChannel}
         ]
        }]
      }]
    }]}
 }
 Collected3 {
  binding_count=1
  length_seconds=1.1226
  object=&self:Collected3
  sequence {bindings [Image_Gem {
     id=#2
     possessed_class=Image
     tracks [{
       type=MovieScene2DTransformTrack
       sections [{
         type=MovieScene2DTransformSection
         end_frame=22
         start_frame=0
         channels [
          Translation.X {type=MovieSceneScriptingFloatChannel, keys [[0,0],[22.451,-100]]}
          Translation.Y {type=MovieSceneScriptingFloatChannel, keys [[0,0],[22.451,-60]]}
          Angle {type=MovieSceneScriptingFloatChannel}
          Scale.X {type=MovieSceneScriptingFloatChannel}
          Scale.Y {type=MovieSceneScriptingFloatChannel}
          Shear.X {type=MovieSceneScriptingFloatChannel}
          Shear.Y {type=MovieSceneScriptingFloatChannel}
         ]
        }]
      }]
    }]}
 }
 Collected4 {
  binding_count=1
  length_seconds=1.1226
  object=&self:Collected4
  sequence {bindings [Image_Gem {
     id=#2
     possessed_class=Image
     tracks [{
       type=MovieScene2DTransformTrack
       sections [{
         type=MovieScene2DTransformSection
         end_frame=22
         start_frame=0
         channels [
          Translation.X {type=MovieSceneScriptingFloatChannel, const [0,2]}
          Translation.Y {type=MovieSceneScriptingFloatChannel, keys [[0,0],[22.451,-115]]}
          Angle {type=MovieSceneScriptingFloatChannel}
          Scale.X {type=MovieSceneScriptingFloatChannel}
          Scale.Y {type=MovieSceneScriptingFloatChannel}
          Shear.X {type=MovieSceneScriptingFloatChannel}
          Shear.Y {type=MovieSceneScriptingFloatChannel}
         ]
        }]
      }]
    }]}
 }
 Collected5 {
  binding_count=1
  length_seconds=1.1226
  object=&self:Collected5
  sequence {bindings [Image_Gem {
     id=#2
     possessed_class=Image
     tracks [{
       type=MovieScene2DTransformTrack
       sections [{
         type=MovieScene2DTransformSection
         end_frame=22
         start_frame=0
         channels [
          Translation.X {type=MovieSceneScriptingFloatChannel, const [0,2]}
          Translation.Y {type=MovieSceneScriptingFloatChannel, keys [[0,0],[22.451,115]]}
          Angle {type=MovieSceneScriptingFloatChannel}
          Scale.X {type=MovieSceneScriptingFloatChannel}
          Scale.Y {type=MovieSceneScriptingFloatChannel}
          Shear.X {type=MovieSceneScriptingFloatChannel}
          Shear.Y {type=MovieSceneScriptingFloatChannel}
         ]
        }]
      }]
    }]}
 }
 Collected6 {
  binding_count=1
  length_seconds=1.1226
  object=&self:Collected6
  sequence {bindings [Image_Gem {
     id=#2
     possessed_class=Image
     tracks [{
       type=MovieScene2DTransformTrack
       sections [{
         type=MovieScene2DTransformSection
         end_frame=22
         start_frame=0
         channels [
          Translation.X {type=MovieSceneScriptingFloatChannel, keys [[0,0],[22.451,100]]}
          Translation.Y {type=MovieSceneScriptingFloatChannel, keys [[0,0],[22.451,60]]}
          Angle {type=MovieSceneScriptingFloatChannel}
          Scale.X {type=MovieSceneScriptingFloatChannel}
          Scale.Y {type=MovieSceneScriptingFloatChannel}
          Shear.X {type=MovieSceneScriptingFloatChannel}
          Shear.Y {type=MovieSceneScriptingFloatChannel}
         ]
        }]
      }]
    }]}
 }
]
graphs [EventGraph {nodes [
   Event Intro() {is=title K2Node_CustomEvent}
   > SetVisibility(self=Get Image_Gem() -> (Image_Gem:Image Object Reference), InVisibility=Visible)
   > PlayAnimation(InAnimation=Get Arrive() -> (Arrive:Widget Animation Object Reference))
   > SetVisibility(self=Get Overlay_Undiscovered() -> (Overlay_Undiscovered:Overlay Object Reference), InVisibility=Hidden)
   Event Outro() {is=title K2Node_CustomEvent}
   > PlayAnimation(InAnimation=Get CollectedBase() -> (CollectedBase:Widget Animation Object Reference))
   n1 = Get StoneNumber() -> (StoneNumber:int=0)
   > PlayAnimation(InAnimation=Select!(
     Option 0:Widget Animation Object Reference=Get Collected1() -> (Collected1:Widget Animation Object Reference)
     Option 1:Widget Animation Object Reference=Get Collected2() -> (Collected2:Widget Animation Object Reference)
     Option 2:Widget Animation Object Reference=Get Collected3() -> (Collected3:Widget Animation Object Reference)
     Option 3:Widget Animation Object Reference=Get Collected4() -> (Collected4:Widget Animation Object Reference)
     Option 4:Widget Animation Object Reference=Get Collected5() -> (Collected5:Widget Animation Object Reference)
     Option 5:Widget Animation Object Reference=Get Collected6() -> (Collected6:Widget Animation Object Reference)
     Index:int=int - int(A=@n1, B=1) {is=title K2Node_CallFunction}
    ) -> (ReturnValue:Widget Animation Object Reference))
   n2 = EventHide() {is=title K2Node_CustomEvent}
   SetVisibility(execute=[@n2.then, @n4.then], self=Get Image_Gem(), InVisibility=Hidden)
   > SetVisibility(self=Get Overlay_Undiscovered(), InVisibility=Visible)
   Event PreConstruct()
   n3 = int - int(A=Get StoneNumber() -> (StoneNumber=0), B=1)
   > SetBrushFromMaterial(
    self=Get Image_Gem()
    Material=Select(
     Option 0:Material Interface Object Reference=/Figment_Martin/Athena/HUD/LTM/FrogParty/Materials/M_UI_FrogParty_RockItem_1
     Option 1:Material Interface Object Reference=/Figment_Martin/Athena/HUD/LTM/FrogParty/Materials/M_UI_FrogParty_RockItem_2
     Option 2:Material Interface Object Reference=/Figment_Martin/Athena/HUD/LTM/FrogParty/Materials/M_UI_FrogParty_RockItem_3
     Option 3:Material Interface Object Reference=/Figment_Martin/Athena/HUD/LTM/FrogParty/Materials/M_UI_FrogParty_RockItem_4
     Option 4:Material Interface Object Reference=/Figment_Martin/Athena/HUD/LTM/FrogParty/Materials/M_UI_FrogParty_RockItem_5
     Option 5:Material Interface Object Reference=/Figment_Martin/Athena/HUD/LTM/FrogParty/Materials/M_UI_FrogParty_RockItem_6
     Index=@n3
    ) -> (ReturnValue:Material Interface Object Reference)
   )
   > SetColorAndOpacity(
    self:Image Object Reference=Get Image_UndiscoveredFill() -> (Image_UndiscoveredFill:Image Object Reference)
    InColorAndOpacity:Linear Color Structure=Select(
     Option 0:Linear Color Structure=(R=0.3686,G=0,B=0.7865,A=1)
     Option 1:Linear Color Structure=(R=0.0369,G=0.5089,B=0.9911,A=1)
     Option 2:Linear Color Structure=(R=0.9647,G=0.063,B=0.063,A=1)
     Option 3:Linear Color Structure=(R=1,G=0.3564,B=0.0027,A=1)
     Option 4:Linear Color Structure=(R=0.147,G=0.8388,B=0.0467,A=1)
     Option 5:Linear Color Structure=(R=0.9948,G=0.9848,B=0,A=1)
     Index=@n3
    ) -> (ReturnValue:Linear Color Structure)
   )
   n4 = Event Construct()
  ]}]
variables [StoneNumber {type=integer, default=1}]
widgets {
 info {inheritedWidgetCount=0, parentClass=&UserWidget, rootWidgetClass=&Overlay, widgetCount=5}
 widgets [
  Overlay_0 {widgetClassPath=&Overlay, widget=Overlay_0}
  Overlay_Undiscovered {
   widgetClassPath=&Overlay
   bIsVariable=true
   parent=Overlay_0
   widget=Overlay_Undiscovered
   slot {class=OverlaySlot, properties {horizontalAlignment=HAlign_Center, verticalAlignment=VAlign_Center}}
  }
  Image_Undiscovered {
   widgetClassPath=&Image
   bIsVariable=true
   parent=Overlay_Undiscovered
   widget=Image_Undiscovered
   properties {
    brush {
     bIsDynamicallyLoaded=false
     drawAs=Image
     imageType=NoImage
     mirroring=NoMirror
     resourceName=None
     resourceObject=null
     tiling=NoTile
     imageSize=<16 16>
     margin=<ltrb 0 0 0 0>
     outlineSettings {
      bUseBrushTransparency=false
      roundingType=HalfHeightRadius
      width=0
      color {colorUseRule=UseColor_Specified, specifiedColor=<rgba 0 0 0 0>}
      cornerRadii=<0 0 0 0>
     }
     tintColor {colorUseRule=UseColor_Specified, specifiedColor=<rgba 1 1 1 1>}
     uVRegion {bIsValid=false, max=<0 0>, min=<0 0>}
    }
    renderTransform {angle=45, scale=<1 1>, shear=<0 0>, translation=<0 0>}
   }
   slot {class=OverlaySlot, properties {horizontalAlignment=HAlign_Center, verticalAlignment=VAlign_Center}}
  }
  Image_UndiscoveredFill {
   widgetClassPath=&Image
   bIsVariable=true
   parent=Overlay_Undiscovered
   widget=Image_UndiscoveredFill
   properties {
    brush {
     bIsDynamicallyLoaded=false
     drawAs=Image
     imageType=NoImage
     mirroring=NoMirror
     resourceName=None
     resourceObject=null
     tiling=NoTile
     imageSize=<12 12>
     margin=<ltrb 0 0 0 0>
     outlineSettings {
      bUseBrushTransparency=false
      roundingType=HalfHeightRadius
      width=0
      color {colorUseRule=UseColor_Specified, specifiedColor=<rgba 0 0 0 0>}
      cornerRadii=<0 0 0 0>
     }
     tintColor {colorUseRule=UseColor_Specified, specifiedColor=<rgba 1 1 1 1>}
     uVRegion {bIsValid=false, max=<0 0>, min=<0 0>}
    }
    renderTransform {angle=45, scale=<1 1>, shear=<0 0>, translation=<0 0>}
   }
   slot {class=OverlaySlot, properties {horizontalAlignment=HAlign_Center, verticalAlignment=VAlign_Center}}
  }
  Image_Gem {
   widgetClassPath=&Image
   bIsVariable=true
   parent=Overlay_0
   widget=Image_Gem
   properties {
    brush {
     bIsDynamicallyLoaded=false
     drawAs=Image
     imageType=NoImage
     mirroring=NoMirror
     resourceName=None
     resourceObject=&M_UI_FrogParty_RockItem_1
     tiling=NoTile
     imageSize=<64 72>
     margin=<ltrb 0 0 0 0>
     outlineSettings {
      bUseBrushTransparency=false
      roundingType=HalfHeightRadius
      width=0
      color {colorUseRule=UseColor_Specified, specifiedColor=<rgba 0 0 0 0>}
      cornerRadii=<0 0 0 0>
     }
     tintColor {colorUseRule=UseColor_Specified, specifiedColor=<rgba 1 1 1 1>}
     uVRegion {bIsValid=false, max=<0 0>, min=<0 0>}
    }
    renderTransform {angle=0, scale=<0.8 0.8>, shear=<0 0>, translation=<0 0>}
   }
   slot {class=OverlaySlot, properties {horizontalAlignment=HAlign_Center, verticalAlignment=VAlign_Center}}
  }
 ]
}
