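# Additive unlit UI material that draws a sharp, tintable cross flare with optional flicker, stored with the Ashton LTM HUD materials.
# A grayscale flare texture blends DarkColor toward white under FlareColor, plus a whitened center glow from a tight sphere mask.
# The texture also drives opacity, modulated by a time-offset sine flicker.
# Quality and feature-level switches drop the glow and near-camera/depth fade on low-end targets.
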Material /Figment_Martin/Athena/HUD/LTM/Ashton/Materials/M_Sharp_Cross_Flare
graph {
 properties {
  bUsedWithParticleSprites=true
  blendMode=BLEND_Additive
  materialDomain=MD_UI
  pixelDepthOffsetMode=PDOM_Legacy
  shadingModel=MSM_Unlit
  usedShadingModels=Unlit
 }
 nodes [
  n1 = VectorParameter() {parameterName=FlareColor, defaultValue=<rgba 1 1 1 1>}
  n2 = LinearInterpolate(A=@n1.RGB) {constAlpha=0.2}
  n3 = Power(Base=SphereMask(A=TextureCoordinate(), B=Constant() {r=0.5}) {attenuationRadius=0.08, hardnessPercent=0}) {constExponent=8}
  n4 = Constant()
  n5 = QualitySwitch(Default=Multiply(A=Multiply(A=@n2, B=@n3)) {constB=2}, Low=@n4)
  n6 = VectorParameter() {parameterName=DarkColor, defaultValue=<rgba 0.1133 0.088 0.2448 1>}
  n7 = TextureSample() {samplerType=SAMPLERTYPE_Grayscale, texture=/Game/Effects/Fort_Effects/Textures/Flares/T_Cross_Flare}
  Add(A=FeatureLevelSwitch(Default=@n5, ES2=@n4, ES3_1=@n4, SM4_REMOVED=@n5), B=Multiply(A=@n1.RGB, B=LinearInterpolate(A=@n6.RGB, Alpha=@n7.R)))
  n8 = ScalarParameter() {defaultValue=150, parameterName=NearFadeDistance}
  n9 = Constant() {r=1}
  n10 = QualitySwitch(
   Default=DepthFade(
    Opacity=MaterialFunctionCall(Distance=@n8) {materialFunction=&Near_Camera_Fade}.Result
    FadeDistance=ScalarParameter() {defaultValue=30, parameterName=DepthFadeDistance}
   )
   Low=@n9
  )
  FeatureLevelSwitch(Default=@n10, ES2=@n9, ES3_1=@n9, SM4_REMOVED=@n10)
  n11 = Time()
  n12 = ScalarParameter() {parameterName=TimeOffset}
  n13 = Fmod(A=Multiply(A=Add(A=@n11, B=@n12), B=ScalarParameter() {parameterName=FlickerSpeed}), B=Constant() {r=1})
  Multiply(
   A=Multiply(A=ScalarParameter() {defaultValue=1, parameterName=Opacity}, B=@n7.R)
   B=LinearInterpolate(
    A=Constant() {r=1}
    B=MaterialFunctionCall(Value=@n13) {materialFunction=&LinearSine}."Rounded Linear Sine "
    Alpha=ScalarParameter() {parameterName=FlickerAmount}
   )
  )
 ]
}
