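# HUD widget for the FrogParty LTM that draws a gem-colored light flare.
# It stacks three centered material images (wide flare, halo corona and rotated sharp cross spike) in an Overlay with negative padding, letting them extend beyond the parent bounds.
# PreConstruct tints all three images with the GemColor variable, white by default, while Construct and Tick are empty.
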WidgetBlueprint /Figment_Martin/Athena/HUD/LTM/FrogParty/HUD_FrogParty_Flare
graphs [EventGraph {nodes [
   Event PreConstruct()
   > SetColorAndOpacity(
    self:Image Object Reference=[
     Get Image_SpikeFlare() -> (Image_SpikeFlare:Image Object Reference)
     Get Image_WideFlare() -> (Image_WideFlare:Image Object Reference)
     Get Image_Corona() -> (Image_Corona:Image Object Reference)
    ]
    InColorAndOpacity:Linear Color Structure=Get GemColor() -> (GemColor:Linear Color Structure)
   )
   Event Construct()
   Event Tick()
  ]}]
variables [GemColor {type=LinearColor, default=<rgba 1 1 1 1>}]
widgets {
 info {inheritedWidgetCount=0, parentClass=&UserWidget, rootWidgetClass=&Overlay, widgetCount=6}
 widgets [
  Overlay_1 {widgetClassPath=&Overlay, widget=Overlay_1}
  Overlay_2 {widgetClassPath=&Overlay, parent=Overlay_1, widget=Overlay_2, slot {class=OverlaySlot}}
  Overlay_0 {
   widgetClassPath=&Overlay
   parent=Overlay_2
   widget=Overlay_0
   slot {class=OverlaySlot, properties {horizontalAlignment=HAlign_Center, verticalAlignment=VAlign_Center, padding=<ltrb -200 -200 -200 -200>}}
  }
  Image_WideFlare {
   widgetClassPath=&Image
   bIsVariable=true
   parent=Overlay_0
   widget=Image_WideFlare
   properties {brush {
     bIsDynamicallyLoaded=false
     drawAs=Image
     imageType=NoImage
     mirroring=NoMirror
     resourceName=None
     resourceObject=/Game/Athena/HUD/Materials/Victory/M_UI_LightFlare
     tiling=NoTile
     imageSize=<169.3371 55.0202>
     margin=<ltrb 0 0 0 0>
     outlineSettings {
      bUseBrushTransparency=false
      roundingType=HalfHeightRadius
      width=0
      color {colorUseRule=UseColor_Specified, specifiedColor=<rgba 0 0 0 0>}
      cornerRadii=<0 0 0 0>
     }
     tintColor {colorUseRule=UseColor_Specified, specifiedColor=<rgba 1 1 1 1>}
     uVRegion {bIsValid=false, max=<0 0>, min=<0 0>}
    }}
   slot {class=OverlaySlot, properties {horizontalAlignment=HAlign_Center, verticalAlignment=VAlign_Center}}
  }
  Image_Corona {
   widgetClassPath=&Image
   bIsVariable=true
   parent=Overlay_0
   widget=Image_Corona
   properties {brush {
     bIsDynamicallyLoaded=false
     drawAs=Image
     imageType=NoImage
     mirroring=NoMirror
     resourceName=None
     resourceObject=/Figment_Martin/Athena/HUD/LTM/Ashton/Materials/M_Flare_Halo_Corona
     tiling=NoTile
     imageSize=<64 64>
     margin=<ltrb 0 0 0 0>
     outlineSettings {
      bUseBrushTransparency=false
      roundingType=HalfHeightRadius
      width=0
      color {colorUseRule=UseColor_Specified, specifiedColor=<rgba 0 0 0 0>}
      cornerRadii=<0 0 0 0>
     }
     tintColor {colorUseRule=UseColor_Specified, specifiedColor=<rgba 1 1 1 1>}
     uVRegion {bIsValid=false, max=<0 0>, min=<0 0>}
    }}
   slot {class=OverlaySlot, properties {horizontalAlignment=HAlign_Center, verticalAlignment=VAlign_Center}}
  }
  Image_SpikeFlare {
   widgetClassPath=&Image
   bIsVariable=true
   parent=Overlay_0
   widget=Image_SpikeFlare
   properties {
    brush {
     bIsDynamicallyLoaded=false
     drawAs=Image
     imageType=NoImage
     mirroring=NoMirror
     resourceName=None
     resourceObject=/Figment_Martin/Athena/HUD/LTM/Ashton/Materials/M_Sharp_Cross_Flare
     tiling=NoTile
     imageSize=<128 128>
     margin=<ltrb 0 0 0 0>
     outlineSettings {
      bUseBrushTransparency=false
      roundingType=HalfHeightRadius
      width=0
      color {colorUseRule=UseColor_Specified, specifiedColor=<rgba 0 0 0 0>}
      cornerRadii=<0 0 0 0>
     }
     tintColor {colorUseRule=UseColor_Specified, specifiedColor=<rgba 1 1 1 1>}
     uVRegion {bIsValid=false, max=<0 0>, min=<0 0>}
    }
    renderTransform {angle=-24, scale=<1 1>, shear=<0 0>, translation=<0 0>}
   }
   slot {class=OverlaySlot, properties {horizontalAlignment=HAlign_Center, verticalAlignment=VAlign_Center}}
  }
 ]
}
