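# Fortnite Frog Party LTM HUD widget, built on AshtonWidgetBase, showing six colored stones around a gauntlet and the villain's remaining lives.
# OnStoneChanged maps type and state to indices into widget arrays built on Construct, hiding or introducing rocks, or playing Outro and revealing socketed gem and flare on capture.
# OnVillainLivesCountChanged drives the count text and dynamic-material progress bar, hidden below zero.
# Hero-lives logic remains, but its widget targets are unconnected.
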WidgetBlueprint /Figment_Martin/Athena/HUD/LTM/FrogParty/HUD_FrogPartyStatus
graphs [EventGraph {nodes [
   n1 = Event OnHeroLivesCountChanged!() -> (OutputDelegate, then, NumLivesLeft:int=0, LivesPercentLeft:float) {is=title K2Node_Event}
   "SetText (Text)"(InText=Format Text!(Format="LOCTEXT(\"x{lives}\")", lives:int=@n1.NumLivesLeft) -> (Result))
   > SetVisibility(InVisibility=Hidden)
   n2 = GetDynamicMaterial()
   > n3 = Set ProgressBarDMI_Hero(ProgressBarDMI_Hero:Material Instance Dynamic Object Reference=@n2) -> (Output_Get:Material Instance Dynamic Object Reference)
   > SetScalarParameterValue(self=@n3, ParameterName=PercentFill, Value=1)
   > SetVisibility(InVisibility=Hidden)
   n4 = Get StoneIndex() -> (StoneIndex:int=0)
   n5 = "Get (a copy)"(Array:Array of Widget Object References=Get GemsAvailable() -> (GemsAvailable:Array of Widget Object References), Dimension 1=@n4) -> (Output:Widget Object Reference)
   n6 = SetVisibility(self=@n5, InVisibility=Hidden)
   n7 = Cast To HUD_FrogPartyStatus_Rock!(execute=[@n6.then, @n20.then], Object=@n5) -> (then, CastFailed, AsHUD Frog Party Status Rock:HUD Frog Party Status Rock Object Reference) {is=title K2Node_DynamicCast}
   > Event Outro(self:HUD Frog Party Status Rock Object Reference=@n7)
   > SetVisibility(
    self=[
     "Get (a copy)"(Array=Get GemFlares() -> (GemFlares:Array of Widget Object References), Dimension 1=@n4)
     "Get (a copy)"(Array=Get GemsSocketed() -> (GemsSocketed:Array of Widget Object References), Dimension 1=@n4)
    ]
    InVisibility=HitTestInvisible
   )
   Event Construct()
   > Sequence()
   > Set GemsSocketed(GemsSocketed:Array of Widget Object References=Make Array!(
     "[0]":Widget Object Reference=Get Image_GemSocketed1() -> (Image_GemSocketed1:Image Object Reference)
     "[1]":Widget Object Reference=Get Image_GemSocketed2() -> (Image_GemSocketed2:Image Object Reference)
     "[2]":Widget Object Reference=Get Image_GemSocketed3() -> (Image_GemSocketed3:Image Object Reference)
     "[3]":Widget Object Reference=Get Image_GemSocketed4() -> (Image_GemSocketed4:Image Object Reference)
     "[4]":Widget Object Reference=Get Image_GemSocketed5() -> (Image_GemSocketed5:Image Object Reference)
     "[5]":Widget Object Reference=Get Image_GemSocketed6() -> (Image_GemSocketed6:Image Object Reference)
    ) -> (Array:Array of Widget Object References)) -> (Output_Get:Array of Widget Object References)
   > Set GemsAvailable(GemsAvailable:Array of Widget Object References=Make Array(
     "[0]"=Get HUD_AshtonStatus_Rock1() -> (HUD_AshtonStatus_Rock1:HUD Frog Party Status Rock Object Reference)
     "[1]"=Get HUD_AshtonStatus_Rock2() -> (HUD_AshtonStatus_Rock2:HUD Frog Party Status Rock Object Reference)
     "[2]"=Get HUD_AshtonStatus_Rock3() -> (HUD_AshtonStatus_Rock3:HUD Frog Party Status Rock Object Reference)
     "[3]"=Get HUD_AshtonStatus_Rock4() -> (HUD_AshtonStatus_Rock4:HUD Frog Party Status Rock Object Reference)
     "[4]"=Get HUD_AshtonStatus_Rock5() -> (HUD_AshtonStatus_Rock5:HUD Frog Party Status Rock Object Reference)
     "[5]"=Get HUD_AshtonStatus_Rock6() -> (HUD_AshtonStatus_Rock6:HUD Frog Party Status Rock Object Reference)
    )) -> (Output_Get:Array of Widget Object References)
   > Set GemFlares(GemFlares:Array of Widget Object References=Make Array(
     "[0]"=Get AshtonFlare_1() -> (AshtonFlare_1:HUD Frog Party Flare Object Reference)
     "[1]"=Get AshtonFlare_2() -> (AshtonFlare_2:HUD Frog Party Flare Object Reference)
     "[2]"=Get AshtonFlare_3() -> (AshtonFlare_3:HUD Frog Party Flare Object Reference)
     "[3]"=Get AshtonFlare_4() -> (AshtonFlare_4:HUD Frog Party Flare Object Reference)
     "[4]"=Get AshtonFlare_5() -> (AshtonFlare_5:HUD Frog Party Flare Object Reference)
     "[5]"=Get AshtonFlare_6() -> (AshtonFlare_6:HUD Frog Party Flare Object Reference)
    )) -> (Output_Get:Array of Widget Object References)
   > n8 = GetDynamicMaterial(self=Get Image_BackplateProgress() -> (Image_BackplateProgress:Image Object Reference))
   > n9 = Set ProgressBarDMI(ProgressBarDMI:Material Instance Dynamic Object Reference=@n8) -> (Output_Get:Material Instance Dynamic Object Reference)
   > SetScalarParameterValue(self=@n9, ParameterName=PercentFill, Value=1)
   n10 = Event OnStoneChanged!() -> (OutputDelegate, then, StoneType:EAshtonStoneType Enum=Purple, StoneState:EAshtonStoneStateType Enum=NotSpawned) {is=title K2Node_Event}
   > n11 = Sequence!(execute) -> (then_0, then_1, then_2:Exec)
   > n12 = Switch on EAshtonStoneType!(execute, Selection:EAshtonStoneType Enum=@n10.StoneType) -> (Purple:Exec, Blue:Exec, Red:Exec, Orange:Exec, Green:Exec, Yellow:Exec) {is=title K2Node_SwitchEnum}
   Set StoneIndex(execute=@n12.Blue, StoneIndex:int=1) -> (Output_Get:int=0)
   Set StoneIndex(execute=@n12.Green, StoneIndex=4) -> (Output_Get=0)
   Set StoneIndex(execute=@n12.Orange, StoneIndex=3) -> (Output_Get=0)
   Set StoneIndex(execute=@n12.Purple, StoneIndex=0) -> (Output_Get=0)
   Set StoneIndex(execute=@n12.Red, StoneIndex=2) -> (Output_Get=0)
   Set StoneIndex(execute=@n12.Yellow, StoneIndex=5) -> (Output_Get=0)
   n13 = Switch on EAshtonStoneStateType!(execute=@n11.then_1, Selection:EAshtonStoneStateType Enum=@n10.StoneState) -> (NotSpawned:Exec, Spawned:Exec, Captured:Exec) {is=title K2Node_SwitchEnum}
   Set StateIndex(execute=@n13.Captured, StateIndex:int=2) -> (Output_Get:int=0)
   Set StateIndex(execute=@n13.NotSpawned, StateIndex=0) -> (Output_Get=0)
   Set StateIndex(execute=@n13.Spawned, StateIndex=1) -> (Output_Get=0)
   n14 = Get StateIndex() -> (StateIndex:int=0)
   n15 = Branch(execute=@n11.then_2, Condition="Equal (Integer)"(A=@n14, B=0) {is=title K2Node_CallFunction})
   > n16 = Cast To HUD_FrogPartyStatus_Rock(Object="Get (a copy)"(Array=Get GemsAvailable(), Dimension 1=@n4))
   > EventHide(self:HUD Frog Party Status Rock Object Reference=@n16)
   n17 = Branch(execute=@n15.else, Condition="Equal (Integer)"(A=@n14, B=1))
   n18 = "Get (a copy)"(Array=Get GemsAvailable(), Dimension 1=@n4)
   > SetVisibility(self=@n18, InVisibility=HitTestInvisible)
   > n19 = Cast To HUD_FrogPartyStatus_Rock(Object=@n18)
   > Event Intro(self:HUD Frog Party Status Rock Object Reference=@n19)
   n20 = Branch(execute=@n17.else, Condition="Equal (Integer)"(A=@n14, B=2))
   n21 = Event OnVillainLivesCountChanged!() -> (OutputDelegate, then, NumLivesLeft:int=0, LivesPercentLeft:float) {is=title K2Node_Event}
   > n22 = Branch(Condition="integer < integer"!(A:int=@n21.NumLivesLeft, B:int=0) -> (ReturnValue:bool) {is=title K2Node_CallFunction})
   > SetVisibility(self=Get Overlay_ProgressBar() -> (Overlay_ProgressBar:Overlay Object Reference), InVisibility=Hidden)
   SetVisibility(execute=@n22.else, self=Get Overlay_ProgressBar(), InVisibility=HitTestInvisible)
   > SetScalarParameterValue(
    self=Get ProgressBarDMI() -> (ProgressBarDMI:Material Instance Dynamic Object Reference)
    ParameterName=PercentFill
    Value=@n21.LivesPercentLeft
   )
   > "SetText (Text)"(
    self=Get ScoreTextBox() -> (ScoreTextBox:Common Text Object Reference)
    InText=Format Text(Format="LOCTEXT(\"x{lives}\")", lives=@n21.NumLivesLeft)
   )
   SetVisibility(execute=@n23.then, InVisibility=HitTestInvisible)
   SetVisibility(InVisibility=Collapsed)
   SetVisibility(execute=@n26.then, InVisibility=Hidden)
   FindParentWidgetOfType!(execute, StartingWidget:Widget Object Reference=Self-Reference(), Type:Widget Class Reference) -> (then, ReturnValue:Widget Object Reference)
   Event PreConstruct()
   Event Tick()
   n23 = SetScalarParameterValue(
    execute=@n25.then
    self=Get ProgressBarDMI_Hero() -> (ProgressBarDMI_Hero:Material Instance Dynamic Object Reference)
    ParameterName=PercentFill
    Value=@n1.LivesPercentLeft
   )
   n24 = "integer < integer"(A=@n1.NumLivesLeft, B=0)
   n25 = SetVisibility(execute=@n26.else, InVisibility=HitTestInvisible)
   n26 = Branch(execute=@n1.then, Condition=@n24)
  ]}]
variables [
 GemFlares {type="array<Widget>", default [null]}
 GemsAvailable {type="array<Widget>", default [null]}
 GemsSocketed {type="array<Widget>", default [null]}
 ProgressBarDMI {type=MaterialInstanceDynamic}
 ProgressBarDMI_Hero {type=MaterialInstanceDynamic}
 StateIndex {type=integer}
 StoneIndex {type=integer}
]
widgets {
 info {inheritedWidgetCount=1, parentClass=/Script/FortniteUI.AshtonWidgetBase, rootWidgetClass=&VerticalBox, widgetCount=38}
 widgets [
  VerticalBox_0 {widgetClassPath=&VerticalBox, widget=VerticalBox_0}
  ScaleBox_0 {
   widgetClassPath=&ScaleBox
   parent=VerticalBox_0
   widget=ScaleBox_0
   properties {stretch=UserSpecified, userSpecifiedScale=0.76}
   slot {class=VerticalBoxSlot, properties {padding=<ltrb 0 8 0 0>}}
  }
  VerticalBox_1 {
   widgetClassPath=&VerticalBox
   parent=ScaleBox_0
   widget=VerticalBox_1
   slot {class=ScaleBoxSlot, properties {horizontalAlignment=HAlign_Fill}}
  }
  Overlay_2 {widgetClassPath=&Overlay, parent=VerticalBox_1, widget=Overlay_2, slot {class=VerticalBoxSlot}}
  Overlay_0 {
   widgetClassPath=&Overlay
   parent=Overlay_2
   widget=Overlay_0
   slot {class=OverlaySlot, properties {horizontalAlignment=HAlign_Center, verticalAlignment=VAlign_Center}}
  }
  Image_Backplate {
   widgetClassPath=&Image
   bIsVariable=true
   parent=Overlay_0
   widget=Image_Backplate
   properties {brush {
     bIsDynamicallyLoaded=false
     drawAs=Image
     imageType=NoImage
     mirroring=NoMirror
     resourceName=None
     resourceObject=/Figment_Martin/Athena/HUD/LTM/FrogParty/Materials/T_UI_FrogParty_PlateBG
     tiling=NoTile
     imageSize=<256 256>
     margin=<ltrb 0 0 0 0>
     outlineSettings {
      bUseBrushTransparency=false
      roundingType=HalfHeightRadius
      width=0
      color {colorUseRule=UseColor_Specified, specifiedColor=<rgba 0 0 0 0>}
      cornerRadii=<0 0 0 0>
     }
     tintColor {colorUseRule=UseColor_Specified, specifiedColor=<rgba 1 1 1 1>}
     uVRegion {bIsValid=false, max=<0 0>, min=<0 0>}
    }}
   slot {class=OverlaySlot}
  }
  Overlay_1 {
   widgetClassPath=&Overlay
   parent=Overlay_0
   widget=Overlay_1
   slot {class=OverlaySlot, properties {horizontalAlignment=HAlign_Center, verticalAlignment=VAlign_Center}}
  }
  Image_Gauntlet {
   widgetClassPath=&Image
   bIsVariable=true
   parent=Overlay_1
   widget=Image_Gauntlet
   properties {brush {
     bIsDynamicallyLoaded=false
     drawAs=Image
     imageType=NoImage
     mirroring=NoMirror
     resourceName=None
     resourceObject=/Figment_Martin/Athena/HUD/LTM/FrogParty/Materials/T_UI_FrogParty_Hand
     tiling=NoTile
     imageSize=<133 171>
     margin=<ltrb 0 0 0 0>
     outlineSettings {
      bUseBrushTransparency=false
      roundingType=HalfHeightRadius
      width=0
      color {colorUseRule=UseColor_Specified, specifiedColor=<rgba 0 0 0 0>}
      cornerRadii=<0 0 0 0>
     }
     tintColor {colorUseRule=UseColor_Specified, specifiedColor=<rgba 1 1 1 1>}
     uVRegion {bIsValid=false, max=<0 0>, min=<0 0>}
    }}
   slot {class=OverlaySlot}
  }
  Image_GemSocketed2 {
   widgetClassPath=&Image
   bIsVariable=true
   parent=Overlay_1
   widget=Image_GemSocketed2
   properties {
    visibility=Hidden
    brush {
     bIsDynamicallyLoaded=false
     drawAs=Image
     imageType=NoImage
     mirroring=NoMirror
     resourceName=None
     resourceObject=/Figment_Martin/Athena/HUD/LTM/FrogParty/Materials/M_UI_FrogPartyRock_2
     tiling=NoTile
     imageSize=<21 22>
     margin=<ltrb 0 0 0 0>
     outlineSettings {
      bUseBrushTransparency=false
      roundingType=HalfHeightRadius
      width=0
      color {colorUseRule=UseColor_Specified, specifiedColor=<rgba 0 0 0 0>}
      cornerRadii=<0 0 0 0>
     }
     tintColor {colorUseRule=UseColor_Specified, specifiedColor=<rgba 1 1 1 1>}
     uVRegion {bIsValid=false, max=<0 0>, min=<0 0>}
    }
    renderTransform {angle=0, scale=<1 1>, shear=<0 0>, translation=<60 43>}
   }
   slot {class=OverlaySlot}
  }
  Image_GemSocketed1 {
   widgetClassPath=&Image
   bIsVariable=true
   parent=Overlay_1
   widget=Image_GemSocketed1
   properties {
    visibility=Hidden
    brush {
     bIsDynamicallyLoaded=false
     drawAs=Image
     imageType=NoImage
     mirroring=NoMirror
     resourceName=None
     resourceObject=/Figment_Martin/Athena/HUD/LTM/FrogParty/Materials/M_UI_FrogPartyRock_1
     tiling=NoTile
     imageSize=<22 22>
     margin=<ltrb 0 0 0 0>
     outlineSettings {
      bUseBrushTransparency=false
      roundingType=HalfHeightRadius
      width=0
      color {colorUseRule=UseColor_Specified, specifiedColor=<rgba 0 0 0 0>}
      cornerRadii=<0 0 0 0>
     }
     tintColor {colorUseRule=UseColor_Specified, specifiedColor=<rgba 1 1 1 1>}
     uVRegion {bIsValid=false, max=<0 0>, min=<0 0>}
    }
    renderTransform {angle=0, scale=<1 1>, shear=<0 0>, translation=<82 46>}
   }
   slot {class=OverlaySlot}
  }
  Image_GemSocketed3 {
   widgetClassPath=&Image
   bIsVariable=true
   parent=Overlay_1
   widget=Image_GemSocketed3
   properties {
    visibility=Hidden
    brush {
     bIsDynamicallyLoaded=false
     drawAs=Image
     imageType=NoImage
     mirroring=NoMirror
     resourceName=None
     resourceObject=/Figment_Martin/Athena/HUD/LTM/FrogParty/Materials/M_UI_FrogPartyRock_3
     tiling=NoTile
     imageSize=<22 22>
     margin=<ltrb 0 0 0 0>
     outlineSettings {
      bUseBrushTransparency=false
      roundingType=HalfHeightRadius
      width=0
      color {colorUseRule=UseColor_Specified, specifiedColor=<rgba 0 0 0 0>}
      cornerRadii=<0 0 0 0>
     }
     tintColor {colorUseRule=UseColor_Specified, specifiedColor=<rgba 1 1 1 1>}
     uVRegion {bIsValid=false, max=<0 0>, min=<0 0>}
    }
    renderTransform {angle=0, scale=<1 1>, shear=<0 0>, translation=<37 43>}
   }
   slot {class=OverlaySlot}
  }
  Image_GemSocketed4 {
   widgetClassPath=&Image
   bIsVariable=true
   parent=Overlay_1
   widget=Image_GemSocketed4
   properties {
    visibility=Hidden
    brush {
     bIsDynamicallyLoaded=false
     drawAs=Image
     imageType=NoImage
     mirroring=NoMirror
     resourceName=None
     resourceObject=/Figment_Martin/Athena/HUD/LTM/FrogParty/Materials/M_UI_FrogPartyRock_4
     tiling=NoTile
     imageSize=<22 22>
     margin=<ltrb 0 0 0 0>
     outlineSettings {
      bUseBrushTransparency=false
      roundingType=HalfHeightRadius
      width=0
      color {colorUseRule=UseColor_Specified, specifiedColor=<rgba 0 0 0 0>}
      cornerRadii=<0 0 0 0>
     }
     tintColor {colorUseRule=UseColor_Specified, specifiedColor=<rgba 1 1 1 1>}
     uVRegion {bIsValid=false, max=<0 0>, min=<0 0>}
    }
    renderTransform {angle=0, scale=<1 1>, shear=<0 0>, translation=<16 48>}
   }
   slot {class=OverlaySlot}
  }
  Image_GemSocketed5 {
   widgetClassPath=&Image
   bIsVariable=true
   parent=Overlay_1
   widget=Image_GemSocketed5
   properties {
    visibility=Hidden
    brush {
     bIsDynamicallyLoaded=false
     drawAs=Image
     imageType=NoImage
     mirroring=NoMirror
     resourceName=None
     resourceObject=/Figment_Martin/Athena/HUD/LTM/FrogParty/Materials/M_UI_FrogPartyRock_5
     tiling=NoTile
     imageSize=<21 22>
     margin=<ltrb 0 0 0 0>
     outlineSettings {
      bUseBrushTransparency=false
      roundingType=HalfHeightRadius
      width=0
      color {colorUseRule=UseColor_Specified, specifiedColor=<rgba 0 0 0 0>}
      cornerRadii=<0 0 0 0>
     }
     tintColor {colorUseRule=UseColor_Specified, specifiedColor=<rgba 1 1 1 1>}
     uVRegion {bIsValid=false, max=<0 0>, min=<0 0>}
    }
    renderTransform {angle=0, scale=<1 1>, shear=<0 0>, translation=<108 76>}
   }
   slot {class=OverlaySlot}
  }
  Image_GemSocketed6 {
   widgetClassPath=&Image
   bIsVariable=true
   parent=Overlay_1
   widget=Image_GemSocketed6
   properties {
    visibility=Hidden
    brush {
     bIsDynamicallyLoaded=false
     drawAs=Image
     imageType=NoImage
     mirroring=NoMirror
     resourceName=None
     resourceObject=/Figment_Martin/Athena/HUD/LTM/FrogParty/Materials/M_UI_FrogPartyRock_6
     tiling=NoTile
     imageSize=<39 39>
     margin=<ltrb 0 0 0 0>
     outlineSettings {
      bUseBrushTransparency=false
      roundingType=HalfHeightRadius
      width=0
      color {colorUseRule=UseColor_Specified, specifiedColor=<rgba 0 0 0 0>}
      cornerRadii=<0 0 0 0>
     }
     tintColor {colorUseRule=UseColor_Specified, specifiedColor=<rgba 1 1 1 1>}
     uVRegion {bIsValid=false, max=<0 0>, min=<0 0>}
    }
    renderTransform {angle=0, scale=<0.99 0.99>, shear=<0 0>, translation=<41 68>}
    renderTransformPivot=<0 0>
   }
   slot {class=OverlaySlot}
  }
  Overlay_5 {
   widgetClassPath=&Overlay
   parent=Overlay_1
   widget=Overlay_5
   slot {class=OverlaySlot, properties {horizontalAlignment=HAlign_Fill, verticalAlignment=VAlign_Fill}}
  }
  AshtonFlare_1 {
   widgetClassPath=&/HUD_FrogParty_Flare
   bIsVariable=true
   parent=Overlay_5
   widget=AshtonFlare_1
   properties {visibility=Hidden, gemColor=<rgba 0.363 0.1485 1 1>, renderTransform {angle=0, scale=<1 1>, shear=<0 0>, translation=<93 54>}}
   slot {class=OverlaySlot}
  }
  AshtonFlare_2 {
   widgetClassPath=&HUD_FrogParty_Flare
   bIsVariable=true
   parent=Overlay_5
   widget=AshtonFlare_2
   properties {visibility=Hidden, gemColor=<rgba 0.0023 0.3292 1 1>, renderTransform {angle=0, scale=<1 1>, shear=<0 0>, translation=<70 49>}}
   slot {class=OverlaySlot}
  }
  AshtonFlare_3 {
   widgetClassPath=&HUD_FrogParty_Flare
   bIsVariable=true
   parent=Overlay_5
   widget=AshtonFlare_3
   properties {visibility=Hidden, gemColor=<rgba 1 0 0.0087 1>, renderTransform {angle=0, scale=<1 1>, shear=<0 0>, translation=<45 52>}}
   slot {class=OverlaySlot}
  }
  AshtonFlare_4 {
   widgetClassPath=&HUD_FrogParty_Flare
   bIsVariable=true
   parent=Overlay_5
   widget=AshtonFlare_4
   properties {visibility=Hidden, gemColor=<rgba 0.776 0.1507 0 1>, renderTransform {angle=0, scale=<1 1>, shear=<0 0>, translation=<26 57>}}
   slot {class=OverlaySlot}
  }
  AshtonFlare_5 {
   widgetClassPath=&HUD_FrogParty_Flare
   bIsVariable=true
   parent=Overlay_5
   widget=AshtonFlare_5
   properties {visibility=Hidden, gemColor=<rgba 0 1 0.1727 1>, renderTransform {angle=0, scale=<1 1>, shear=<0 0>, translation=<116 84>}}
   slot {class=OverlaySlot}
  }
  AshtonFlare_6 {
   widgetClassPath=&HUD_FrogParty_Flare
   bIsVariable=true
   parent=Overlay_5
   widget=AshtonFlare_6
   properties {visibility=Hidden, gemColor=<rgba 1 0.8954 0.0166 1>, renderTransform {angle=0, scale=<1 1>, shear=<0 0>, translation=<52 84>}}
   slot {class=OverlaySlot}
  }
  OverlayGemsAvailable {
   widgetClassPath=&Overlay
   parent=Overlay_0
   widget=OverlayGemsAvailable
   slot {class=OverlaySlot, properties {horizontalAlignment=HAlign_Fill, verticalAlignment=VAlign_Fill}}
  }
  HUD_AshtonStatus_Rock1 {
   widgetClassPath=&/HUD_FrogPartyStatus_Rock
   bIsVariable=true
   parent=OverlayGemsAvailable
   widget=HUD_AshtonStatus_Rock1
   properties {renderTransform {angle=0, scale=<1 1>, shear=<0 0>, translation=<102 -57>}}
   slot {class=OverlaySlot, properties {horizontalAlignment=HAlign_Center, verticalAlignment=VAlign_Center}}
  }
  HUD_AshtonStatus_Rock2 {
   widgetClassPath=&HUD_FrogPartyStatus_Rock
   bIsVariable=true
   parent=OverlayGemsAvailable
   widget=HUD_AshtonStatus_Rock2
   properties {stoneNumber=2, renderTransform {angle=0, scale=<1 1>, shear=<0 0>, translation=<-102 55>}}
   slot {class=OverlaySlot, properties {horizontalAlignment=HAlign_Center, verticalAlignment=VAlign_Center}}
  }
  HUD_AshtonStatus_Rock3 {
   widgetClassPath=&HUD_FrogPartyStatus_Rock
   bIsVariable=true
   parent=OverlayGemsAvailable
   widget=HUD_AshtonStatus_Rock3
   properties {stoneNumber=3, renderTransform {angle=0, scale=<1 1>, shear=<0 0>, translation=<102 55>}}
   slot {class=OverlaySlot, properties {horizontalAlignment=HAlign_Center, verticalAlignment=VAlign_Center}}
  }
  HUD_AshtonStatus_Rock4 {
   widgetClassPath=&HUD_FrogPartyStatus_Rock
   bIsVariable=true
   parent=OverlayGemsAvailable
   widget=HUD_AshtonStatus_Rock4
   properties {stoneNumber=4, renderTransform {angle=0, scale=<1 1>, shear=<0 0>, translation=<0 114>}}
   slot {class=OverlaySlot, properties {horizontalAlignment=HAlign_Center, verticalAlignment=VAlign_Center}}
  }
  HUD_AshtonStatus_Rock5 {
   widgetClassPath=&HUD_FrogPartyStatus_Rock
   bIsVariable=true
   parent=OverlayGemsAvailable
   widget=HUD_AshtonStatus_Rock5
   properties {stoneNumber=5, renderTransform {angle=0, scale=<1 1>, shear=<0 0>, translation=<0 -114>}}
   slot {class=OverlaySlot, properties {horizontalAlignment=HAlign_Center, verticalAlignment=VAlign_Center}}
  }
  HUD_AshtonStatus_Rock6 {
   widgetClassPath=&HUD_FrogPartyStatus_Rock
   bIsVariable=true
   parent=OverlayGemsAvailable
   widget=HUD_AshtonStatus_Rock6
   properties {stoneNumber=6, renderTransform {angle=0, scale=<1 1>, shear=<0 0>, translation=<-102 -57>}}
   slot {class=OverlaySlot, properties {horizontalAlignment=HAlign_Center, verticalAlignment=VAlign_Center}}
  }
  Overlay_ProgressBar {
   widgetClassPath=&Overlay
   bIsVariable=true
   parent=VerticalBox_1
   widget=Overlay_ProgressBar
   slot {class=VerticalBoxSlot, properties {horizontalAlignment=HAlign_Center, verticalAlignment=VAlign_Top, padding=<ltrb 0 -5 0 0>}}
  }
  Image_BackplateProgress {
   widgetClassPath=&Image
   bIsVariable=true
   parent=Overlay_ProgressBar
   widget=Image_BackplateProgress
   properties {
    bFlipForRightToLeftFlowDirection=true
    brush {
     bIsDynamicallyLoaded=false
     drawAs=Box
     imageType=NoImage
     mirroring=NoMirror
     resourceName=None
     resourceObject=/Figment_Martin/Athena/HUD/Materials/Score/M_ScoreLTM_ScoreProgress_Purple
     tiling=NoTile
     imageSize=<32 54.4196>
     margin=<ltrb 0.09 0.09 0.09 0.09>
     outlineSettings {
      bUseBrushTransparency=false
      roundingType=HalfHeightRadius
      width=0
      color {colorUseRule=UseColor_Specified, specifiedColor=<rgba 0 0 0 0>}
      cornerRadii=<0 0 0 0>
     }
     tintColor {colorUseRule=UseColor_Specified, specifiedColor=<rgba 1 1 1 1>}
     uVRegion {bIsValid=false, max=<0 0>, min=<0 0>}
    }
    renderTransform {angle=0, scale=<1 1>, shear=<-5 0>, translation=<0 0>}
   }
   slot {class=OverlaySlot, properties {horizontalAlignment=HAlign_Fill, verticalAlignment=VAlign_Center, padding=<ltrb 35 0 0 0>}}
  }
  Overlay_3 {
   widgetClassPath=&Overlay
   parent=Overlay_ProgressBar
   widget=Overlay_3
   slot {class=OverlaySlot, properties {horizontalAlignment=HAlign_Fill, verticalAlignment=VAlign_Fill}}
  }
  ScaleBox_1 {
   widgetClassPath=&ScaleBox
   parent=Overlay_3
   widget=ScaleBox_1
   properties {stretch=UserSpecified, userSpecifiedScale=0.8}
   slot {class=OverlaySlot, properties {verticalAlignment=VAlign_Center}}
  }
  OverlayPlacePlate {
   widgetClassPath=&Overlay
   bIsVariable=true
   parent=ScaleBox_1
   widget=OverlayPlacePlate
   properties {renderTransform {angle=0, scale=<1 1>, shear=<0 0>, translation=<0 2>}}
   slot {class=ScaleBoxSlot}
  }
  Image_Generic {
   widgetClassPath=&Image
   parent=OverlayPlacePlate
   widget=Image_Generic
   properties {
    brush {
     bIsDynamicallyLoaded=false
     drawAs=Image
     imageType=NoImage
     mirroring=NoMirror
     resourceName=None
     resourceObject=/Figment_Martin/Athena/HUD/LTM/FrogParty/Materials/T_UI_FrogParty_EnemyIcon
     tiling=NoTile
     imageSize=<70 93>
     margin=<ltrb 0 0 0 0>
     outlineSettings {
      bUseBrushTransparency=false
      roundingType=HalfHeightRadius
      width=0
      color {colorUseRule=UseColor_Specified, specifiedColor=<rgba 0 0 0 0>}
      cornerRadii=<0 0 0 0>
     }
     tintColor {colorUseRule=UseColor_Specified, specifiedColor=<rgba 1 1 1 1>}
     uVRegion {bIsValid=false, max=<0 0>, min=<0 0>}
    }
    renderTransform {angle=0, scale=<0.8 0.8>, shear=<0 0>, translation=<0 -4>}
   }
   slot {class=OverlaySlot, properties {horizontalAlignment=HAlign_Fill, verticalAlignment=VAlign_Fill}}
  }
  SizeBoxScore {
   widgetClassPath=&SizeBox
   parent=Overlay_3
   widget=SizeBoxScore
   properties {bOverride_WidthOverride=true, widthOverride=308}
   slot {class=OverlaySlot, properties {horizontalAlignment=HAlign_Fill, verticalAlignment=VAlign_Center}}
  }
  ScoreOverlay {widgetClassPath=&Overlay, parent=SizeBoxScore, widget=ScoreOverlay, slot {class=SizeBoxSlot}}
  ScoreTextBox {
   widgetClassPath=&CommonTextBlock
   bIsVariable=true
   parent=ScoreOverlay
   widget=ScoreTextBox
   properties {
    text=x100
    font {
     bForceMonospaced=false
     bMaterialIsStencil=false
     fontMaterial=null
     fontObject=&NotoSans
     letterSpacing=0
     monospacedWidth=1
     size=20
     skewAmount=0
     typefaceFontName=NotoSans-Bold
     outlineSettings {
      bApplyOutlineToDropShadows=false
      bMiteredCorners=false
      bSeparateFillAlpha=false
      outlineMaterial=null
      outlineSize=1
      outlineColor=<rgba 0 0 0 0.8>
     }
    }
    renderTransformPivot=<0.2 0.8>
   }
   slot {class=OverlaySlot, properties {horizontalAlignment=HAlign_Center, verticalAlignment=VAlign_Center, padding=<ltrb 0 -10 0 -10>}}
  }
  TutorialHighlight {widgetClassPath=&TutorialHighlight}
 ]
}
