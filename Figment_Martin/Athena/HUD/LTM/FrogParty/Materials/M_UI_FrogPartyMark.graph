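# M_UI_FrogPartyMark is a translucent UI-domain material, apparently a mark for the Frog Party LTM HUD.
# It holds a texture sample of a Season 8 Ashton banner icon and a constant of one.
# No links from either node to each other or to the material outputs are present, so the graph appears to be a stub.
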Material /Figment_Martin/Athena/HUD/LTM/FrogParty/Materials/M_UI_FrogPartyMark
graph {
 properties {blendMode=BLEND_Translucent, materialDomain=MD_UI, pixelDepthOffsetMode=PDOM_Legacy, usedShadingModels=Default Lit}
 nodes [
  Constant() {r=1}
  TextureSample() {texture=/BRCosmetics/2dAssets/Banners/Season8/T-Banners-Icons-S8-Ashton-A}
 ]
}
