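# Translucent UI material for a rock image in the Frog Party LTM HUD.
# A radial exponential gradient takes scaled UVs, a center, and Radius and Density parameters.
# Scale and Offset each feed only their red channel into both axes, so scaling and centering are uniform.
# A clamped texture parameter defaults to the rock texture, with no links shown from it or the gradient to material outputs.
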Material /Figment_Martin/Athena/HUD/LTM/FrogParty/Materials/M_UI_FrogPartyRock
graph {
 properties {blendMode=BLEND_Translucent, materialDomain=MD_UI, pixelDepthOffsetMode=PDOM_Legacy, usedShadingModels=Default Lit}
 nodes [
  n1 = VectorParameter() {parameterName=Scale, defaultValue=<rgba 1 1 0 1>}
  n2 = VectorParameter() {parameterName=Offset, defaultValue=<rgba 0.5 0.5 0 1>}
  MaterialFunctionCall(
   UVs=Multiply(A=TextureCoordinate(), B=AppendVector(A=@n1.R, B=@n1.R))
   CenterPosition=AppendVector(A=@n2.R, B=@n2.R)
   Radius=ScalarParameter() {defaultValue=0.5, parameterName=Radius}
   Density=ScalarParameter() {defaultValue=2.3, parameterName=Density}
  ) {materialFunction=&RadialGradientExponential}
  TextureSampleParameter2D() {
   parameterName=Texture
   samplerSource=SSM_Clamp_WorldGroupSettings
   texture=&/T_UI_FrogParty_Rock1
  }
 ]
}
