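# Translucent UI material that draws a rock item icon for the Frog Party LTM HUD, with a shine and a whiteout effect.
# The item texture lerps toward white by a Whiteout parameter and serves as the base of a Color Dodge blend.
# The blend layer is the shared MF_Shine function, driven by offset and rotation parameters and masked by the rounded texture alpha.
# The unrounded texture alpha also passes through as an output.
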Material /Figment_Martin/Athena/HUD/LTM/FrogParty/Materials/M_UI_FrogParty_RockItem
graph {
 properties {blendMode=BLEND_Translucent, materialDomain=MD_UI, pixelDepthOffsetMode=PDOM_Legacy, usedShadingModels=Default Lit}
 nodes [
  n1 = ScalarParameter() {defaultValue=0.54, parameterName=ShineOffset}
  n2 = ScalarParameter() {defaultValue=0.6, parameterName=Rotation}
  n3 = Constant2Vector() {g=0.5, r=0.5}
  n4 = TextureSampleParameter2D() {parameterName=Texture, texture=&/T_UI_FrogParty_RockItem1}
  MaterialFunctionCall!(
   Blend=Multiply(
    A=MaterialFunctionCall!(ShineOffset=@n1, Rotation=@n2, RotationCenter=@n3, UVs) {materialFunction=/Game/Athena/HUD/Materials/Victory/MF_Shine}.Result
    B=Round(None=@n4.A)
   )
   Base=LinearInterpolate(A=@n4.RGB, B=Constant() {r=1}, Alpha=ScalarParameter() {parameterName=Whiteout})
  ) {materialFunction=/Engine/Functions/Engine_MaterialFunctions03/Blends/Blend_ColorDodge}
  Reroute(None=@n4.A)
 ]
}
